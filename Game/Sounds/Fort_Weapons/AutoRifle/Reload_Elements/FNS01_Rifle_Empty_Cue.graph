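# Weapon foley sound cue for an auto rifle running empty, stored among the rifle's reload elements.
# A Mixer layers an assault-end wave, raised to a fixed higher pitch by a Modulator, with a quieter negative UI gameplay sound.
# The cue plays at reduced volume in the weapon foley sound class, with attenuation falloff set and a concurrency override that blocks new instances once the limit is reached.
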SoundCue /Game/Sounds/Fort_Weapons/AutoRifle/Reload_Elements/FNS01_Rifle_Empty_Cue
root=Mixer_2
nodes [
 {id=Mixer_2, children [Modulator_1,WavePlayer_9], properties {inputVolume [1,0.4]}}
 {id=Modulator_1, children [WavePlayer_10], properties {pitchMax=1.3, pitchMin=1.3}}
 {id=WavePlayer_10, properties {soundWaveAssetPtr=&/assault_end_sdk_01}}
 {id=WavePlayer_9, properties {soundWaveAssetPtr=/Game/Sounds/Fort_UI_Sounds/Gameplay/fort_ui_new_neg_01}}
]
properties {
 bOverrideConcurrency=true
 maxDistance=2097152
 soundClassObject=&Fort_WeaponFoley
 volumeMultiplier=0.3
 attenuationOverrides {falloffDistance=1600}
 concurrencyOverrides {resolutionRule=PreventNew}
}
